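# Physics asset for the SteelyKendama back bling skeletal mesh, defining collision and limited swing for its dangling gong, rings and tassel.
# A capsule on the pack bone anchors a boxed gong and three small sphere rings, each constrained directly to it.
# A three-capsule tassel chain hangs from the gong, with swing limits widening toward the tip.
# Twist is locked on the gong and held very tight on the rings and tassel links.
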PhysicsAsset /BRCosmetics/Accessories/FORT_Backpacks/Backpack_M_MED_SteelyKendama/Meshes/M_MED_SteelyKendama_Pack_Physics
bodies {
 dyn_gong [{
   shapeType=Box
   extentX=4.2129
   extentY=16.6544
   extentZ=18.5904
   shapeName=dyn_gong_box
   center=<10.6984 -0.6384 -0.0034>
   rotation=<rot -0.0004 90.858 -89.9786>
  }]
 dyn_ring_A [{shapeType=Sphere, radius=1.2, shapeName=dyn_ring_A_sphere, center=<0.8756 0.0088 0.0062>}]
 dyn_ring_B [{shapeType=Sphere, radius=1.1, shapeName=dyn_ring_B_sphere, center=<0.9833 -0.0846 -0.0242>}]
 dyn_ring_C [{shapeType=Sphere, radius=1.1, shapeName=dyn_ring_C_sphere, center=<0.8573 -0.0757 -0.0792>}]
 dyn_tassel_01 [{shapeType=Capsule, length=0.3948, radius=1.1696, shapeName=dyn_tassel_01_capsule, center=<1.2704 0 0.0551>, rotation=<rot 0 -90 -90>}]
 dyn_tassel_02 [{shapeType=Capsule, length=0.4432, radius=1.4334, shapeName=dyn_tassel_02_capsule, center=<2.0511 0 0.0263>, rotation=<rot 0 -90 -90>}]
 dyn_tassel_03 [{shapeType=Capsule, length=1.6919, radius=2.8903, shapeName=dyn_tassel_03_capsule, center=<3.622 0 -0.2719>, rotation=<rot 0 -90 -90>}]
 pack [{
   shapeType=Capsule
   length=28.2126
   radius=12.013
   shapeName=pack_capsule
   center=<-0.6211 12.5232 -5.7328>
   rotation=<rot 0.7122 88.2766 178.6248>
  }]
}
constraints [
 {bone1Name=dyn_gong, bone2Name=pack, swing1LimitDegrees=25, swing2LimitDegrees=2, twistLimitDegrees=0, twistMotion=Locked}
 {bone1Name=dyn_tassel_01, bone2Name=dyn_gong, swing1LimitDegrees=10, swing2LimitDegrees=10, twistLimitDegrees=1}
 {bone1Name=dyn_tassel_02, bone2Name=dyn_tassel_01, swing1LimitDegrees=30, swing2LimitDegrees=25, twistLimitDegrees=1}
 {bone1Name=dyn_tassel_03, bone2Name=dyn_tassel_02, swing1LimitDegrees=35, swing2LimitDegrees=45, twistLimitDegrees=1}
 {bone1Name=dyn_ring_A, bone2Name=pack, swing1LimitDegrees=40, swing2LimitDegrees=15, twistLimitDegrees=1}
 {bone1Name=dyn_ring_B, bone2Name=pack, swing1LimitDegrees=60, swing2LimitDegrees=30, twistLimitDegrees=1}
 {bone1Name=dyn_ring_C, bone2Name=pack, swing1LimitDegrees=40, swing2LimitDegrees=15, twistLimitDegrees=1}
]
